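; Doorway wall Blueprint for the EchoFrost set, with adjustable ice on its doors and a corrected interact icon position.
; The construction script finds static mesh components using either of two EchoFrost door meshes and writes the ice amount variable into their default custom primitive data at index 2.
; GetIconPlacement returns the mesh's center of mass lowered along Z, centering the mobile Easy Interact icon on the doors (FORT-1063857).
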(fn GetIconPlacement
  (params (SelfActor :type Actor%20Object%20Reference) (ViewingActor :type Actor%20Object%20Reference))
  ; FORT-1063857 - Fixes the Easy Interact Icon's placement on Mobile so that it is centered in the middle of the doors.
  (return (- (Physics|GetCenterOfMass (Variables|Debug|GetStaticMeshComponent) "None") "0.000000,0.000000,50.000000") "0, 0, 0"))

(fn ConstructionScript
  (for _array_element (Actor|GetComponentsByClass "/Script/Engine.StaticMeshComponent")
    (if (or (== (Variables|/Script/Engine.StaticMeshComponent|GetStaticMesh _array_element) "/Hera/Sets/EchoFrost/Meshes/SM_EchoFrost_Door_B_A.SM_EchoFrost_Door_B_A") (== (Variables|/Script/Engine.StaticMeshComponent|GetStaticMesh _array_element) "/Hera/Sets/EchoFrost/Meshes/SM_EchoFrost_Door_A_A.SM_EchoFrost_Door_A_A"))
      (Rendering|Material|SetDefaultCustomPrimitiveDataFloat _array_element 2 (Variables|Default|GetDoors%20Ice%20Amount%20%28%22CPD%22%20Offset%29)))))
